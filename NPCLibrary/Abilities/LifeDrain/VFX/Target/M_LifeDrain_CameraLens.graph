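# Additive, unlit particle material for the Life Drain ability's camera lens effect on the target.
# Noise distorts scrolling flame textures, combined with inky smoke and a vertical gradient-sine mask.
# RGB-offset, oversaturated samples of flame and scene color add chromatic fringing; lower quality, forward and mobile paths replace scene color with a constant.
# A curve-atlas gradient or Color parameter, partly desaturated, tints the emissive, with particle color and alpha applied.
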Material /NPCLibrary/Abilities/LifeDrain/VFX/Target/M_LifeDrain_CameraLens
graph {
 properties {
  bUseTranslucencyVertexFog=false
  bUsedWithMeshParticles=true
  bUsedWithNiagaraMeshParticles=true
  bUsedWithNiagaraSprites=true
  blendMode=BLEND_Additive
  pixelDepthOffsetMode=PDOM_Legacy
  shadingModel=MSM_Unlit
  twoSided=true
  usedShadingModels=Unlit
 }
 nodes [
  n1 = CurveAtlasRowParameter(CurveTime=MaterialFunctionCall(Float2=TextureCoordinate()) {materialFunction=&BreakOutFloat2Components}.G) {
   atlas=/NPCLibrary/Abilities/LifeDrain/VFX/Source/CurveAtlas_LifeDrain
   curve=/NPCLibrary/Abilities/LifeDrain/VFX/Source/Curve_LifeDrain_Beam_Color
  }
  n2 = Desaturation(
   None=LinearInterpolate(
    A=VectorParameter() {parameterName=Color, defaultValue=<rgba 1 1 1 1>}.RGB
    B=@n1
    Alpha=ScalarParameter() {defaultValue=1, parameterName=UseCurveColor}
   )
   Fraction=Constant() {r=0.3}
  )
  n3 = ComponentMask(None=SceneColor!(OffsetFraction) {inputMode=OffsetFraction, constInput=<0.002 0.002>}.RGB) {r=true}
  n4 = ComponentMask(None=SceneColor() {inputMode=OffsetFraction}.RGB) {g=true}
  n5 = ComponentMask(None=SceneColor() {inputMode=OffsetFraction, constInput=<-0.002 -0.002>}.RGB) {b=true}
  n6 = Multiply(A=Desaturation(
    None=MaterialFunctionCall(X=@n3, Y=@n4, Z=@n5) {materialFunction=&MakeFloat3}.Result
    Fraction=Constant() {r=-1}
   )) {constA=1, constB=5}
  n7 = Constant() {r=0.5}
  n8 = QualitySwitch(Default=@n6, Low=@n7, High=@n6, Medium=@n7, Epic=@n6)
  n9 = Panner(Coordinate=Multiply(A=TextureCoordinate(), B=Constant2Vector() {g=0.0333, r=2})) {speedY=-0.005}
  n10 = TextureCoordinate() {uTiling=2, vTiling=0.15}
  n11 = Add(
   A=Multiply(A=TextureSample!(UVs=@n9, Tex, Level, Apply View MipBias) {mipValueMode=TMVM_MipLevel, samplerType=SAMPLERTYPE_Masks, texture=/Game/Weapons/WeaponSkins/Wraps/Textures/Mash/Mash_Noise}.R) {constB=0.15}
   B=Multiply(A=TextureSample!(UVs=@n10, Tex, Apply View MipBias) {texture=&T_Inky_Smoke_Tile}.R) {constB=0.1}
  )
  n12 = MaterialFunctionCall(Float2=TextureCoordinate()) {materialFunction=&BreakOutFloat2Components}
  n13 = OneMinus(None=@n12.G)
  n14 = MaterialFunctionCall(X=@n12.R, Y=OneMinus(None=Power(Base=@n13) {constExponent=12.6056})) {materialFunction=&MakeFloat2}
  n15 = Constant2Vector() {g=0.019, r=5}
  n16 = Constant2Vector() {g=0.1429}
  n17 = Time()
  n18 = Add(A=@n11, B=Add(A=Multiply(A=@n14.Result, B=@n15), B=Multiply(A=@n16, B=@n17)))
  n19 = TextureSample!(UVs=Add(A=@n18) {constB=0.005}, Tex, Level, Apply View MipBias) {
   mipValueMode=TMVM_MipLevel
   samplerType=SAMPLERTYPE_LinearGrayscale
   texture=/Game/Weapons/FORT_Melee/Pickaxe_Cube_Ninja_Male/FX/T_FluidFlameTile
  }
  n20 = TextureSample(UVs=@n18) {
   mipValueMode=TMVM_MipLevel
   samplerType=SAMPLERTYPE_LinearGrayscale
   texture=&T_FluidFlameTile
  }
  n21 = TextureSample(UVs=Add(A=@n18) {constB=-0.005}) {
   mipValueMode=TMVM_MipLevel
   samplerType=SAMPLERTYPE_LinearGrayscale
   texture=&T_FluidFlameTile
  }
  n22 = OneMinus(None=MaterialFunctionCall() {materialFunction=&LinearGradient}.VGradient)
  n23 = Add(A=@n11, B=TextureCoordinate())
  n24 = Multiply(A=MaterialFunctionCall(Float2=Add(A=@n23) {constB=-0.05}) {materialFunction=&BreakOutFloat2Components}.G) {constB=0.5}
  n25 = Multiply(A=Multiply(A=@n22, B=@n22), B=Power(Base=Sine(None=@n24)) {constExponent=3})
  n26 = Multiply(
   A=Desaturation(
    None=MaterialFunctionCall(X=@n19.R, Y=@n20.R, Z=@n21.R) {materialFunction=&MakeFloat3}.Result
    Fraction=Constant() {r=-1}
   )
   B=@n25
  )
  n27 = MaterialFunctionCall(Float2=TextureCoordinate() {uTiling=3, vTiling=0.5}) {materialFunction=&BreakOutFloat2Components}
  n28 = TextureSample!(
   UVs=MaterialFunctionCall(X=@n27.R, Y=OneMinus(None=@n27.G)) {materialFunction=&MakeFloat2}.Result
   Tex
   Apply View MipBias
  ) {texture=&T_Inky_Smoke_Tile}
  n29 = Multiply(A=ShadingPathSwitch(Default=@n8, Deferred=@n8, Forward=@n7, Mobile=@n7), B=Add(A=Add(A=@n26, B=@n28.R), B=@n25))
  n30 = ParticleColor()
  MaterialFunctionCall(Emissive=Multiply(A=Multiply(A=@n2, B=@n29), B=@n30.RGB)) {materialFunction=&MF_FX_EmissiveAdjustments}
  n31 = Add(A=@n20.R, B=@n28.R)
  Multiply(A=@n30.A, B=Add(A=Multiply(A=@n31, B=@n25), B=@n25))
 ]
}
